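# Niagara effect for the SonnetSpirit emote, adding blue magic glows, lightning and swipe debris around what appears to be a snowman being conjured.
# Lightning stretches meshes between points sampled on the owner skeleton and system skeletal mesh, scaled by their distance.
# Bottom, Middle and Top emitters stagger bone-attached, layered sprite glows, with a light on Bottom.
# SmileSwipe later bursts gravity-pulled, speed-stretched meshes from the Top emitter's bone.
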NiagaraSystem /BRCosmetics/Animation/Game/MainPlayer/Emotes/SonnetSpirit/FX/NS_Emote_SonnetSpirit_MagicGlow
dependencies {
 usedDataInterfaces [&NiagaraDataInterfaceArrayInt32, &NiagaraDataInterfaceCurve, &NiagaraDataInterfaceArrayFloat]
 usedDynamicInputs []
 usedModules [
  &self:SystemSpawnScript.SystemScriptSource.SystemScriptGraph.NiagaraNodeAssignment_0.SetVariables_#1
  &SystemState
  &self:SystemSpawnScript.SystemScriptSource.SystemScriptGraph.NiagaraNodeAssignment_1.SetVariables_#2
 ]
}
emitters {
 Lightning {
  emitterUpdateScript [
   &EmitterState()
   &SpawnBurst_Instantaneous(Spawn Count=2, Spawn Time=0.027, Use Spawn Probability=false, UseLoopCountLimit=false)
   /SetVariables(Emitter.U Offset:NiagaraFloat=&FloatFromCurve(FloatCurve=DataInterface() {propertyValues {Curve {keys [<at 0 2>,<at 1 0>]}}}, CurveIndex=$Emitter.NormalizedLoopAge)) {isSetParameters=true}
  ]
  particleSpawnScript [
   &PartitionParticles()
   &InitialMeshOrientation(Mesh Orientation Mode=::NewEnumerator2, Orientation Vector=<0 0 1>, Rotation=<0 0.5 0>, Use Rotation Vector=false)
   &InitializeParticle(
    Color=&Multiply_LinearColorByFloat(Linear Color=<rgba 0.5469 1.1953 1.5 1>, Float=1)
    Lifetime=0.25
    Mesh Scale Mode=::NewEnumerator4
    Sprite Size Mode=::NewEnumerator3
    UsePositionOffset=false
    Write Material Random=true
   )
  ]
  particleUpdateScript [
   &ParticleState()
   &SkeletalMeshLocation(
    Bone / Socket Index=$Particles.PartitionParticles.Partition
    Bone Sampling Mode=::NewEnumerator4
    MaskSpawnGroup=false
    Orientation Sampling=::NewEnumerator1
    Skeletal Mesh=$User.OwnerSkeleton
   ) {module=SkeletalMeshLocation001}
   &SkeletalMeshLocation(
    Bone Sampling Mode=::NewEnumerator4
    MaskSpawnGroup=false
    Orientation Sampling=::NewEnumerator1
    Position Sampling=::NewEnumerator2
    Skeletal Mesh=$System.Skeletal Mesh
   )
   &ScaleMeshSize(Scale Factor=&Multiply_Vector(
     A=&MakeVector(
      X=1
      Y=1
      Z=&Multiply_Float(
       A=&FloatFromCurve(FloatCurve=DataInterface() {propertyValues {Curve {keys [
            {interpMode=RCIM_Cubic, leaveTangentWeight=0.0714, tangentMode=RCTM_User, time=0, value=0.2}
            {arriveTangentWeight=0.0191, interpMode=RCIM_Cubic, leaveTangentWeight=0.1243, tangentMode=RCTM_User, time=0.0211, value=1}
            {arriveTangentWeight=0.0969, interpMode=RCIM_Cubic, leaveTangentWeight=0.3745, tangentMode=RCTM_User, time=0.1582, value=1}
           ]}}})
       B=&RemapRange(
        Input Value To Remap=&VectorLength(Vector=&Subtract_Position(A=$Particles.SkeletalMeshLocation001.SampledPosition, B=$Particles.SkeletalMeshLocation.SampledPosition))
        Input Min=0
        Input Max=128.513
        Output Min=0
        Output Max=2.5
       )
      )
     )
     B=&Divide_Vector(A:Vector3f=<1 1 1>, B:Vector3f=$Engine.Owner.Scale)
    ))
   &DynamicMaterialParameters(
    Index 0 Param 1=&FloatFromCurve(FloatCurve=DataInterface() {propertyValues {Curve {keys [<at 0 1>,<at 0.4 0>]}}})
    Index 0 Param 3=&FloatFromCurve(FloatCurve=DataInterface() {propertyValues {Curve {keys [
         {arriveTangent=0.2581, interpMode=RCIM_Cubic, leaveTangent=0.2581, tangentMode=RCTM_User, time=0.95, value=0}
         {arriveTangent=2.8205, interpMode=RCIM_Cubic, leaveTangent=2.8205, tangentMode=RCTM_User, time=1, value=1}
        ]}}})
    Index 0 Param 4=5
    Index 1 Param 1=1
    Index 2 Param 2=1
    Index 2 Param 3=1
    Index 2 Param 4=1
    Module.Index0Float0Write=false
    Module.Index0Float1Write=true
    Module.Index0Float2Write=true
    Module.Index0Float3Write=true
    Module.Index1Float0Write=true
    Module.Index1Float1Write=true
    Module.Index1Float2Write=true
    Module.Index1Float3Write=true
    Module.Index2Float0Write=true
    Module.Index2Float1Write=true
    Module.Index2Float2Write=true
    Module.Index2Float3Write=true
    Module.Index3Float0Write=true
    Module.Index3Float1Write=true
    Module.Index3Float2Write=true
    Module.Index3Float3Write=true
    Write Parameter Index 1=::NewEnumerator1
    Write Parameter Index 2=::NewEnumerator1
    Write Parameter Index 3=::NewEnumerator1
   ) {module=DynamicMaterialParameters001}
   &UpdateMeshOrientation(
    Facing Direction=&Subtract_Position(A=$Particles.SkeletalMeshLocation001.SampledPosition, B=$Particles.SkeletalMeshLocation.SampledPosition)
    Facing Interpolation=false
    Orientation Method=::NewEnumerator1
    Override Quat=false
    TransformFacing=false
    TransformUp=false
   ) {module=UpdateMeshOrientation001}
  ]
  rendererClasses [&NiagaraMeshRendererProperties]
  renderers [{rendererClass=&NiagaraMeshRendererProperties, rendererIndex=0}]
 }
 SmileSwipe {
  emitterUpdateScript [
   &EmitterState()
   &SpawnBurst_Instantaneous(Spawn Count=8, Spawn Group=1, Spawn Time=2.35, Use Spawn Probability=false, UseLoopCountLimit=false)
  ]
  particleSpawnScript [
   /SetVariables(Particles.Intensity:NiagaraFloat=&FloatFromCurve(
     FloatCurve=DataInterface() {propertyValues {Curve {keys [<at 0 0>,<at 0.8 0>,<at 1 1>]}}}
     CurveIndex=&RandomRangeFloat(Minimum=0, Maximum=1)
    )) {isSetParameters=true}
   &PartitionParticles(Number Of Partitions=1)
   &InitializeParticle(
    Color=&FNTimeOfDayColor(
     Night Color=<rgba 0.1874 0.3998 0.6719 1>
     Morning Color=<rgba 0.5736 0.7208 1 1>
     Day Color=<rgba 0.5736 0.7208 1 1>
     Evening Color=<rgba 0.5736 0.7208 1 1>
    )
    Lifetime Max=0.35
    Lifetime Min=0.15
    Lifetime Mode=::NewEnumerator1
    Mesh Scale Mode=::NewEnumerator3
    Mesh Uniform Scale=&MakeFloatFromVector(
     Vector=&Multiply_VectorByFloat(
      Vector=&RandomRangeVector(Minimum=<1 0.5 0.5>, Maximum=<1.5 1 1>)
      Float=&Lerp_Float(A=0.009, B=0.01, Alpha=$Particles.Intensity)
     )
     Channel=::NewEnumerator0
    )
    Sprite Rotation Mode=::NewEnumerator2
    Sprite Size Mode=::NewEnumerator3
    UsePositionOffset=false
    Write Material Random=true
   )
   &SkeletalMeshLocation(
    Bone / Socket Index=&SelectIntFromArray(Direct Array Index=$Particles.PartitionParticles.Partition, Int Selection Array=DataInterface() {propertyValues {IntData [6]}})
    Bone Sampling Mode=::NewEnumerator4
    MaskSpawnGroup=false
    Orientation Sampling=::NewEnumerator1
    Sampled Position Offset=<-7.0755 10.6576 0>
    Skeletal Mesh=$System.Skeletal Mesh
   )
   &AddVelocity(
    Cone Angle=35
    Cone Axis=<1 0.718 0>
    Module.UseConeFalloff=false
    Module.UseVelDistribution=false
    Override Seed=false
    Velocity Mode=::NewEnumerator2
    Velocity Speed=&RandomRangeFloat(Minimum=100, Maximum=200)
   )
   &CameraOffset()
  ]
  particleUpdateScript [
   &ParticleState()
   &Drag(Drag=&RandomRangeFloat(Minimum=0.3, Maximum=0.5), Use Linear Drag=true, Use Rotational Drag=false)
   &SpriteRotationRate(Rotation Rate=&SelectFloatFromArray(
     Array Sampling Mode=::NewEnumerator0
     Direct Array Index=?
     Float Selection Array=DataInterface() {propertyValues {FloatData [-150,150,-100,100]}}
     Random Seed=0
    ))
   &GravityForce(Gravity=&RandomRangeVector(Minimum=<0 0 -150>, Maximum=<0 0 -200>))
   &ScaleMeshSize(Scale Factor=&Multiply_VectorByFloat(
     Vector=<1 1 1>
     Float=&FloatFromCurve(FloatCurve=DataInterface() {propertyValues {Curve {keys [
          {arriveTangent=0.6971, interpMode=RCIM_Cubic, leaveTangent=0.6971, tangentMode=RCTM_User, time=0, value=0}
          {arriveTangent=-0.0529, interpMode=RCIM_Cubic, leaveTangent=-0.0529, tangentMode=RCTM_User, time=0.2, value=1.2}
          {arriveTangent=-1.6278, interpMode=RCIM_Cubic, leaveTangent=-1.6278, tangentMode=RCTM_User, time=0.8, value=0.65}
          {arriveTangent=-0.0405, interpMode=RCIM_Cubic, leaveTangent=-0.0405, tangentMode=RCTM_User, time=1, value=0}
         ]}}})
    ))
   &SolveForcesAndVelocity(Clamp Velocity=false, Limit Acceleration=false)
   &ScaleMeshSizeBySpeed(Max Scale Factor=<5 1 0>)
  ]
  rendererClasses [&NiagaraMeshRendererProperties]
  renderers [{rendererClass=&NiagaraMeshRendererProperties, rendererIndex=0}]
 }
 SnowManSpawn_Bottom {
  emitterUpdateScript [
   &EmitterState(Life Cycle Mode=::NewEnumerator1, Loop Behavior=::NewEnumerator1, Loop Duration=0.6, UseLoopDelay=false)
   &SpawnBurst_Instantaneous(Spawn Group=2, Use Spawn Probability=false, UseLoopCountLimit=false) {module=SpawnBurst_Instantaneous001}
   &SpawnBurst_Instantaneous(Spawn Group=1, Use Spawn Probability=false, UseLoopCountLimit=false)
   &SpawnRate(SpawnRate=20, Use Spawn Probability=false)
  ]
  particleSpawnScript [
   &PartitionParticles(Number Of Partitions=1)
   &InitializeParticle(
    Color=&Multiply_LinearColorByFloat(Linear Color=<rgba 0.625 0.8801 1 1>, Float=1)
    Lifetime=&MergeHelper_FloatSwitch(0=0.55, 1=1.1, 2=1.15, 3=?, Index=&SpawnGroup(), Num=::NewEnumerator1)
    Mesh Scale Mode=::NewEnumerator3
    Mesh Uniform Scale=&RandomRangeFloat(Minimum=1.5, Maximum=2.2)
    Sprite Rotation Mode=::NewEnumerator2
    Sprite Size Mode=::NewEnumerator1
    Uniform Sprite Size Max=90
    Uniform Sprite Size Min=80
    UsePositionOffset=false
    Write Material Random=true
   )
   &CameraOffset(Camera Offset Amount=&MergeHelper_FloatSwitch(0=-30, 1=-30, 2=?, 3=?, Index=&SpawnGroup(), Num=::NewEnumerator0))
   /SetVariables(
    Particles.Magic Large Color:LinearColor=<rgba 0.0883 0.3465 0.7708 1>
    Particles.Magic Large Size:Vector2f=&Vector2DFromFloat(Value=&RandomRangeFloat(Minimum=100, Maximum=110))
   ) {isSetParameters=true, module=SetVariables_#3}
   /SetVariables(Particles.Glow Color=<rgba 0.4869 0.7172 1 0.5>, Particles.Glow Size=&Vector2DFromFloat(Value=180)) {isSetParameters=true}
   /SetVariables(Particles.VisibilityTag=&SpawnGroup()) {isSetParameters=true}
  ]
  particleUpdateScript [
   &ParticleState()
   &SkeletalMeshLocation(
    Bone / Socket Index=&SelectIntFromArray(Direct Array Index=$Particles.PartitionParticles.Partition, Int Selection Array=DataInterface() {propertyValues {IntData [0]}})
    Bone Sampling Mode=::NewEnumerator4
    MaskSpawnGroup=false
    Orientation Sampling=::NewEnumerator1
    Sampled Position Offset=&MergeHelper_VectorSwitch(
     0=&RandomRangeVector(Minimum=<-5 -5 25>, Maximum=<5 5 25>)
     1=<0 0 25>
     2=<0 0 65>
     3=?
     Index=&SpawnGroup()
     Num=::NewEnumerator1
    )
    Skeletal Mesh=$System.Skeletal Mesh
   )
   &ScaleSpriteSize(Uniform Curve Sprite Scale=DataInterface() {propertyValues {Curve {keys [
        <at 0 0 RCIM_Cubic>
        {arriveTangentWeight=0.1067, interpMode=RCIM_Cubic, leaveTangentWeight=0.3202, tangentMode=RCTM_User, time=0.3, value=1}
        {arriveTangentWeight=0.3202, interpMode=RCIM_Cubic, leaveTangentWeight=0.1067, tangentMode=RCTM_User, time=0.85, value=1}
        <at 1 0 RCIM_Cubic>
       ]}}})
   &SpriteRotationRate(Rotation Rate=&SelectFloatFromArray(
     Array Sampling Mode=::NewEnumerator0
     Direct Array Index=?
     Float Selection Array=DataInterface() {propertyValues {FloatData [-150,150,-100,100]}}
     Random Seed=0
    ))
   &SolveForcesAndVelocity(Clamp Velocity=false, Limit Acceleration=false)
   /SetVariables(
    Particles.Scale Magic Large Color:LinearColor=&MakeLinearColorFromVectorAndFloat(
     "Vector (RGB)"=&MakeVectorFromLinearColorRGB(Color=$Particles.Initial.Magic Large Color)
     "Float (Alpha)"=&FloatFromCurve(FloatCurve=DataInterface() {propertyValues {Curve {keys [<at 0 1>,<at 1 1>]}}})
    )
    Particles.Scale Magic Large Size:Vector2f=&Vector2DFromFloat(Value=&Multiply_Float(
      A=&MakeFloatFromVector2d(Vector 2D=$Particles.Initial.Magic Large Size)
      B=&FloatFromCurve(FloatCurve=DataInterface() {propertyValues {Curve {keys [
           <at 0 0 RCIM_Cubic>
           {arriveTangentWeight=0.1067, interpMode=RCIM_Cubic, leaveTangentWeight=0.3202, tangentMode=RCTM_User, time=0.4, value=1}
           {arriveTangentWeight=0.3202, interpMode=RCIM_Cubic, leaveTangentWeight=0.1067, tangentMode=RCTM_User, time=0.85, value=1}
           <at 1 0 RCIM_Cubic>
          ]}}})
     ))
   ) {isSetParameters=true, module=SetVariables_#4}
   /SetVariables(
    Particles.Scale Glow Color:LinearColor=&MakeLinearColorFromVectorAndFloat(
     "Vector (RGB)"=&MakeVectorFromLinearColorRGB(Color=$Particles.Initial.Glow Color)
     "Float (Alpha)"=&FloatFromCurve(
      FloatCurve=DataInterface() {propertyValues {Curve {keys [
          <at 0 0 RCIM_Cubic>
          {arriveTangentWeight=0.1035, interpMode=RCIM_Cubic, leaveTangentWeight=0.3227, tangentMode=RCTM_User, time=0.1, value=1}
          {arriveTangentWeight=0.3204, interpMode=RCIM_Cubic, leaveTangentWeight=0.1078, tangentMode=RCTM_User, time=0.7969, value=1}
          <at 1 0 RCIM_Cubic>
         ]}}}
      Scale Curve=0.1
     )
    )
    Particles.Scale Glow Size:Vector2f=&Vector2DFromFloat(Value=&Multiply_Float(
      A=&MakeFloatFromVector2d(Vector 2D=$Particles.Initial.Glow Size)
      B=&FloatFromCurve(FloatCurve=DataInterface() {propertyValues {Curve {keys [
           <at 0 0 RCIM_Cubic>
           {arriveTangentWeight=0.1067, interpMode=RCIM_Cubic, leaveTangentWeight=0.3202, tangentMode=RCTM_User, time=0.4, value=1}
           {arriveTangentWeight=0.3202, interpMode=RCIM_Cubic, leaveTangentWeight=0.1067, tangentMode=RCTM_User, time=0.85, value=1}
           <at 1 0 RCIM_Cubic>
          ]}}})
     ))
   ) {isSetParameters=true}
   &FN_Light_Attributes(
    Color=<rgba 1.1667 2.875 8 1>
    Create Color Attribute=true
    Create Exponent Attribute=true
    Create Radius Attribute=true
    Enable Visibility Tag=true
    Exponent=1
    Radius=90
    Scale Color=&FloatFromCurve(FloatCurve=DataInterface() {propertyValues {Curve {keys [
         <at 0 0 RCIM_Cubic>
         {arriveTangentWeight=0.1035, interpMode=RCIM_Cubic, leaveTangentWeight=0.3227, tangentMode=RCTM_User, time=0.05, value=1}
         {arriveTangentWeight=0.3204, interpMode=RCIM_Cubic, leaveTangentWeight=0.1078, tangentMode=RCTM_User, time=0.8, value=1}
         <at 1 0 RCIM_Cubic>
        ]}}})
   )
  ]
  rendererClasses [&NiagaraSpriteRendererProperties, &NiagaraLightRendererProperties]
  renderers [
   {rendererClass=&NiagaraSpriteRendererProperties}
   {rendererClass=&NiagaraSpriteRendererProperties, rendererIndex=1}
   {rendererClass=&NiagaraSpriteRendererProperties, rendererIndex=2}
   {rendererClass=&NiagaraLightRendererProperties, rendererIndex=3}
  ]
 }
 SnowManSpawn_Middle {
  emitterUpdateScript [
   &EmitterState(Life Cycle Mode=::NewEnumerator1, Loop Behavior=::NewEnumerator1, Loop Delay=0.25, Loop Duration=0.46, UseLoopDelay=true)
   &SpawnBurst_Instantaneous(Spawn Group=1, Use Spawn Probability=false, UseLoopCountLimit=false)
   &SpawnRate(SpawnRate=18, Use Spawn Probability=false)
  ]
  particleSpawnScript [
   &PartitionParticles(Number Of Partitions=1)
   &InitializeParticle(
    Color=&Multiply_LinearColorByFloat(Linear Color=<rgba 0.625 0.8801 1 1>, Float=1)
    Lifetime=&MergeHelper_FloatSwitch(0=0.4, 1=0.85, 2=?, 3=?, Index=&SpawnGroup(), Num=::NewEnumerator0)
    Mesh Scale Mode=::NewEnumerator3
    Mesh Uniform Scale=&RandomRangeFloat(Minimum=1.5, Maximum=2.2)
    Sprite Rotation Mode=::NewEnumerator2
    Sprite Size Mode=::NewEnumerator1
    Uniform Sprite Size Max=70
    Uniform Sprite Size Min=60
    UsePositionOffset=false
    Write Material Random=true
   )
   &CameraOffset(Camera Offset Amount=&MergeHelper_FloatSwitch(0=-30, 1=10, 2=?, 3=?, Index=&SpawnGroup(), Num=::NewEnumerator0))
   /SetVariables(
    Particles.Magic Large Color:LinearColor=<rgba 0.0883 0.3465 0.7708 1>
    Particles.Magic Large Size:Vector2f=&Vector2DFromFloat(Value=85)
   ) {isSetParameters=true, module=SetVariables_#3}
   /SetVariables(Particles.Glow Color=<rgba 0.4869 0.7172 1 0.5>, Particles.Glow Size=&Vector2DFromFloat(Value=160)) {isSetParameters=true}
   /SetVariables(Particles.VisibilityTag=&SpawnGroup()) {isSetParameters=true}
  ]
  particleUpdateScript [
   &ParticleState()
   &SkeletalMeshLocation(
    Bone / Socket Index=&SelectIntFromArray(Direct Array Index=$Particles.PartitionParticles.Partition, Int Selection Array=DataInterface() {propertyValues {IntData [1]}})
    Bone Sampling Mode=::NewEnumerator4
    MaskSpawnGroup=false
    Orientation Sampling=::NewEnumerator1
    Sampled Position Offset=&MergeHelper_VectorSwitch(
     0=&RandomRangeVector(Minimum=<-10 -10 8>, Maximum=<10 10 8>)
     1=<0 0 8>
     2=?
     3=?
     Index=&SpawnGroup()
     Num=::NewEnumerator0
    )
    Skeletal Mesh=$System.Skeletal Mesh
   )
   &ScaleSpriteSize(Uniform Curve Sprite Scale=DataInterface() {propertyValues {Curve {keys [
        <at 0 0 RCIM_Cubic>
        {arriveTangentWeight=0.1067, interpMode=RCIM_Cubic, leaveTangentWeight=0.3202, tangentMode=RCTM_User, time=0.3, value=1}
        {arriveTangentWeight=0.3202, interpMode=RCIM_Cubic, leaveTangentWeight=0.1067, tangentMode=RCTM_User, time=0.85, value=1}
        <at 1 0 RCIM_Cubic>
       ]}}})
   &SpriteRotationRate(Rotation Rate=&SelectFloatFromArray(
     Array Sampling Mode=::NewEnumerator0
     Direct Array Index=?
     Float Selection Array=DataInterface() {propertyValues {FloatData [-150,150,-100,100]}}
     Random Seed=0
    ))
   &SolveForcesAndVelocity(Clamp Velocity=false, Limit Acceleration=false)
   /SetVariables(
    Particles.Scale Magic Large Color:LinearColor=&MakeLinearColorFromVectorAndFloat(
     "Vector (RGB)"=&MakeVectorFromLinearColorRGB(Color=$Particles.Initial.Magic Large Color)
     "Float (Alpha)"=&FloatFromCurve(FloatCurve=DataInterface() {propertyValues {Curve {keys [<at 0 1>,<at 1 1>]}}})
    )
    Particles.Scale Magic Large Size:Vector2f=&Vector2DFromFloat(Value=&Multiply_Float(
      A=&MakeFloatFromVector2d(Vector 2D=$Particles.Initial.Magic Large Size)
      B=&FloatFromCurve(FloatCurve=DataInterface() {propertyValues {Curve {keys [
           <at 0 0 RCIM_Cubic>
           {arriveTangentWeight=0.1067, interpMode=RCIM_Cubic, leaveTangentWeight=0.3202, tangentMode=RCTM_User, time=0.4, value=1}
           {arriveTangentWeight=0.3202, interpMode=RCIM_Cubic, leaveTangentWeight=0.1067, tangentMode=RCTM_User, time=0.85, value=1}
           <at 1 0 RCIM_Cubic>
          ]}}})
     ))
   ) {isSetParameters=true, module=SetVariables_#4}
   /SetVariables(
    Particles.Scale Glow Color:LinearColor=&MakeLinearColorFromVectorAndFloat(
     "Vector (RGB)"=&MakeVectorFromLinearColorRGB(Color=$Particles.Initial.Glow Color)
     "Float (Alpha)"=&FloatFromCurve(
      FloatCurve=DataInterface() {propertyValues {Curve {keys [
          <at 0 0 RCIM_Cubic>
          {arriveTangentWeight=0.1035, interpMode=RCIM_Cubic, leaveTangentWeight=0.3227, tangentMode=RCTM_User, time=0.1, value=1}
          {arriveTangentWeight=0.3204, interpMode=RCIM_Cubic, leaveTangentWeight=0.1078, tangentMode=RCTM_User, time=0.7969, value=1}
          <at 1 0 RCIM_Cubic>
         ]}}}
      Scale Curve=0.1
     )
    )
    Particles.Scale Glow Size:Vector2f=&Vector2DFromFloat(Value=&Multiply_Float(
      A=&MakeFloatFromVector2d(Vector 2D=$Particles.Initial.Glow Size)
      B=&FloatFromCurve(FloatCurve=DataInterface() {propertyValues {Curve {keys [
           <at 0 0 RCIM_Cubic>
           {arriveTangentWeight=0.1067, interpMode=RCIM_Cubic, leaveTangentWeight=0.3202, tangentMode=RCTM_User, time=0.4, value=1}
           {arriveTangentWeight=0.3202, interpMode=RCIM_Cubic, leaveTangentWeight=0.1067, tangentMode=RCTM_User, time=0.85, value=1}
           <at 1 0 RCIM_Cubic>
          ]}}})
     ))
   ) {isSetParameters=true}
  ]
  rendererClasses [&NiagaraSpriteRendererProperties]
  renderers [
   {rendererClass=&NiagaraSpriteRendererProperties}
   {rendererClass=&NiagaraSpriteRendererProperties, rendererIndex=1}
   {rendererClass=&NiagaraSpriteRendererProperties, rendererIndex=2}
  ]
 }
 SnowManSpawn_Top {
  emitterUpdateScript [
   &EmitterState(Life Cycle Mode=::NewEnumerator1, Loop Behavior=::NewEnumerator1, Loop Delay=0.52, Loop Duration=0.2, UseLoopDelay=true)
   &SpawnBurst_Instantaneous(Spawn Group=1, Use Spawn Probability=false, UseLoopCountLimit=false)
   &SpawnRate(SpawnRate=18, Use Spawn Probability=false)
  ]
  particleSpawnScript [
   &PartitionParticles(Number Of Partitions=1)
   &InitializeParticle(
    Color=&Multiply_LinearColorByFloat(Linear Color=<rgba 0.625 0.8801 1 1>, Float=1)
    Lifetime=&MergeHelper_FloatSwitch(0=0.5, 1=0.7, 2=?, 3=?, Index=&SpawnGroup(), Num=::NewEnumerator0)
    Mesh Scale Mode=::NewEnumerator1
    Mesh Uniform Scale Max=1.4
    Mesh Uniform Scale Min=1.2
    Sprite Rotation Mode=::NewEnumerator2
    Sprite Size Mode=::NewEnumerator3
    Uniform Sprite Size=45
    UsePositionOffset=false
    Write Material Random=true
   )
   &CameraOffset(Camera Offset Amount=&MergeHelper_FloatSwitch(0=-30, 1=10, 2=?, 3=?, Index=&SpawnGroup(), Num=::NewEnumerator0))
   /SetVariables(
    Particles.Magic Large Color:LinearColor=<rgba 0.0883 0.3465 0.7708 1>
    Particles.Magic Large Size:Vector2f=&Vector2DFromFloat(Value=55)
   ) {isSetParameters=true}
   /SetVariables(Particles.Glow Color=<rgba 0.4869 0.7172 1 0.5>, Particles.Glow Size=&Vector2DFromFloat(Value=110)) {isSetParameters=true}
   /SetVariables(Particles.VisibilityTag=&SpawnGroup()) {isSetParameters=true}
  ]
  particleUpdateScript [
   &ParticleState()
   &SkeletalMeshLocation(
    Bone / Socket Index=&SelectIntFromArray(Direct Array Index=$Particles.PartitionParticles.Partition, Int Selection Array=DataInterface() {propertyValues {IntData [6]}})
    Bone Sampling Mode=::NewEnumerator4
    MaskSpawnGroup=false
    Orientation Sampling=::NewEnumerator1
    Sampled Position Offset=&MergeHelper_VectorSwitch(
     0=&RandomRangeVector(Minimum=<0 0 5>, Maximum=<0 0 5>)
     1=<0 0 5>
     2=?
     3=?
     Index=&SpawnGroup()
     Num=::NewEnumerator0
    )
    Skeletal Mesh=$System.Skeletal Mesh
   )
   &ScaleSpriteSize(Uniform Curve Sprite Scale=DataInterface() {propertyValues {Curve {keys [
        <at 0 0 RCIM_Cubic>
        {arriveTangentWeight=0.1067, interpMode=RCIM_Cubic, leaveTangentWeight=0.3202, tangentMode=RCTM_User, time=0.3, value=1}
        {arriveTangentWeight=0.3202, interpMode=RCIM_Cubic, leaveTangentWeight=0.1067, tangentMode=RCTM_User, time=0.85, value=1}
        <at 1 0 RCIM_Cubic>
       ]}}})
   &SpriteRotationRate(Rotation Rate=&SelectFloatFromArray(
     Array Sampling Mode=::NewEnumerator0
     Direct Array Index=?
     Float Selection Array=DataInterface() {propertyValues {FloatData [-150,150,-100,100]}}
     Random Seed=0
    ))
   &SolveForcesAndVelocity(Clamp Velocity=false, Limit Acceleration=false)
   /SetVariables(
    Particles.Scale Magic Large Color:LinearColor=&MakeLinearColorFromVectorAndFloat(
     "Vector (RGB)"=&MakeVectorFromLinearColorRGB(Color=$Particles.Initial.Magic Large Color)
     "Float (Alpha)"=&FloatFromCurve(FloatCurve=DataInterface() {propertyValues {Curve {keys [<at 0 1>,<at 1 1>]}}})
    )
    Particles.Scale Magic Large Size:Vector2f=&Vector2DFromFloat(Value=&Multiply_Float(
      A=&MakeFloatFromVector2d(Vector 2D=$Particles.Initial.Magic Large Size)
      B=&FloatFromCurve(FloatCurve=DataInterface() {propertyValues {Curve {keys [
           <at 0 0 RCIM_Cubic>
           {arriveTangentWeight=0.1067, interpMode=RCIM_Cubic, leaveTangentWeight=0.3202, tangentMode=RCTM_User, time=0.4, value=1}
           {arriveTangentWeight=0.3202, interpMode=RCIM_Cubic, leaveTangentWeight=0.1067, tangentMode=RCTM_User, time=0.85, value=1}
           <at 1 0 RCIM_Cubic>
          ]}}})
     ))
   ) {isSetParameters=true}
   /SetVariables(
    Particles.Scale Glow Color:LinearColor=&MakeLinearColorFromVectorAndFloat(
     "Vector (RGB)"=&MakeVectorFromLinearColorRGB(Color=$Particles.Initial.Glow Color)
     "Float (Alpha)"=&FloatFromCurve(
      FloatCurve=DataInterface() {propertyValues {Curve {keys [
          <at 0 0 RCIM_Cubic>
          {arriveTangentWeight=0.1035, interpMode=RCIM_Cubic, leaveTangentWeight=0.3227, tangentMode=RCTM_User, time=0.1, value=1}
          {arriveTangentWeight=0.3204, interpMode=RCIM_Cubic, leaveTangentWeight=0.1078, tangentMode=RCTM_User, time=0.7969, value=1}
          <at 1 0 RCIM_Cubic>
         ]}}}
      Scale Curve=0.1
     )
    )
    Particles.Scale Glow Size:Vector2f=&Vector2DFromFloat(Value=&Multiply_Float(
      A=&MakeFloatFromVector2d(Vector 2D=$Particles.Initial.Glow Size)
      B=&FloatFromCurve(FloatCurve=DataInterface() {propertyValues {Curve {keys [
           <at 0 0 RCIM_Cubic>
           {arriveTangentWeight=0.1067, interpMode=RCIM_Cubic, leaveTangentWeight=0.3202, tangentMode=RCTM_User, time=0.4, value=1}
           {arriveTangentWeight=0.3202, interpMode=RCIM_Cubic, leaveTangentWeight=0.1067, tangentMode=RCTM_User, time=0.85, value=1}
           <at 1 0 RCIM_Cubic>
          ]}}})
     ))
   ) {isSetParameters=true}
  ]
  rendererClasses [&NiagaraSpriteRendererProperties]
  renderers [
   {rendererClass=&NiagaraSpriteRendererProperties}
   {rendererClass=&NiagaraSpriteRendererProperties, rendererIndex=1}
   {rendererClass=&NiagaraSpriteRendererProperties, rendererIndex=2}
  ]
 }
}
summary {
 systemName=NS_Emote_SonnetSpirit_MagicGlow
 emitters [
  {emitterName=SnowManSpawn_Top, rendererClasses [&NiagaraSpriteRendererProperties]}
  {emitterName=Lightning, rendererClasses [&NiagaraMeshRendererProperties]}
  {
   emitterName=SnowManSpawn_Bottom
   rendererClasses [&NiagaraSpriteRendererProperties, &NiagaraLightRendererProperties]
  }
  {emitterName=SnowManSpawn_Middle, rendererClasses [&NiagaraSpriteRendererProperties]}
  {emitterName=SmileSwipe, rendererClasses [&NiagaraMeshRendererProperties]}
 ]
 userVariables [
  User.bIsFrontEnd {type=NiagaraBool}
  User.bIsFrontEndPreview {type=NiagaraBool}
  User.IsLargeBody {type=NiagaraBool}
  User.OwnerSkeleton {
   type=NiagaraDataInterfaceSkeletalMesh
   defaultValue {
    struct=NiagaraExt_VariableValue_DataInterface
    value {dataInterface=&self:NiagaraDataInterfaceSkeletalMesh_0}
   }
  }
 ]
}
