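# Niagara system for the SonnetSpirit emote that puts short-lived blue magic sprites on the character's hands.
# Mirrored left and right emitters spawn continuously and re-sample their own bone or socket on the skeletal mesh during update, so sprites follow the hands.
# Each emitter has two sprite renderers and separate "Magic Large" color and size attributes for a larger layer.
# Sprites spin and scale in and out, with base color dimmed by time of day.
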NiagaraSystem /BRCosmetics/Animation/Game/MainPlayer/Emotes/SonnetSpirit/FX/NS_Emote_SonnetSpirit_MagicHands
dependencies {
 usedDataInterfaces [&NiagaraDataInterfaceCurve, &NiagaraDataInterfaceArrayInt32]
 usedDynamicInputs []
 usedModules [
  &self:SystemSpawnScript.SystemScriptSource.SystemScriptGraph.NiagaraNodeAssignment_0.SetVariables_#1
  &SystemState
 ]
}
emitters {
 MagicHand_L {
  emitterUpdateScript [
   &EmitterState()
   &SpawnBurst_Instantaneous(Use Spawn Probability=false, UseLoopCountLimit=false) {enabled=false}
   &SpawnRate(SpawnRate=18, Use Spawn Probability=false)
  ]
  particleSpawnScript [
   &PartitionParticles(Number Of Partitions=1)
   &InitializeParticle(
    Color=&Multiply_LinearColorByFloat(
     Linear Color=<rgba 0.3073 0.7786 1 1>
     Float=&FNTimeOfDayScalar(Night Scale=0.5, Morning Scale=0.78, Evening Scale=0.78)
    )
    Lifetime=0.25
    Sprite Rotation Mode=::NewEnumerator2
    Sprite Size Mode=::NewEnumerator1
    Uniform Sprite Size Max=22
    Uniform Sprite Size Min=20
    UsePositionOffset=false
    Write Material Random=true
   )
   &CameraOffset(Camera Offset Amount=&RandomRangeFloat(Minimum=-15, Maximum=-10))
   /SetVariables(
    Particles.Magic Large Color:LinearColor=<rgba 0.1214 0.2597 0.9323 1>
    Particles.Magic Large Size:Vector2f=&Vector2DFromFloat(Value=&RandomRangeFloat(Minimum=25, Maximum=30))
   ) {isSetParameters=true}
  ]
  particleUpdateScript [
   &ParticleState()
   &ScaleSpriteSize(Uniform Curve Sprite Scale=DataInterface() {propertyValues {Curve {keys [
        {arriveTangentWeight=0.0807, interpMode=RCIM_Cubic, leaveTangentWeight=0.3949, tangentMode=RCTM_User, time=0, value=0}
        {arriveTangent=1.1137, interpMode=RCIM_Cubic, leaveTangent=1.1137, time=0.15, value=1}
        {arriveTangentWeight=0.3423, interpMode=RCIM_Cubic, leaveTangentWeight=0.0493, tangentMode=RCTM_User, time=0.8979, value=1}
        <at 1 0 RCIM_Cubic>
       ]}}})
   &SpriteRotationRate(Rotation Rate=&RandomRangeFloat(Minimum=-100, Maximum=100))
   &SkeletalMeshLocation(
    Bone / Socket Index=&SelectIntFromArray(Direct Array Index=$Particles.PartitionParticles.Partition, Int Selection Array=DataInterface() {propertyValues {IntData [0]}})
    Bone Sampling Mode=::NewEnumerator4
    MaskSpawnGroup=false
    Orientation Sampling=::NewEnumerator1
    Sampled Position Offset=<10 0 0>
    Skeletal Mesh=$System.Skeletal Mesh
   )
   &SolveForcesAndVelocity(Clamp Velocity=false, Limit Acceleration=false)
   /SetVariables(
    Particles.Scale Magic Large Color:LinearColor=&MakeLinearColorFromVectorAndFloat(
     "Vector (RGB)"=&MakeVectorFromLinearColorRGB(Color=$Particles.Initial.Magic Large Color)
     "Float (Alpha)"=&FloatFromCurve(FloatCurve=DataInterface() {propertyValues {Curve {keys [<at 0 1>,<at 1 1>]}}})
    )
    Particles.Scale Magic Large Size:Vector2f=&Vector2DFromFloat(Value=&Multiply_Float(
      A=&MakeFloatFromVector2d(Vector 2D=$Particles.Initial.Magic Large Size)
      B=&FloatFromCurve(FloatCurve=DataInterface() {propertyValues {Curve {keys [
           {arriveTangentWeight=0.0807, interpMode=RCIM_Cubic, leaveTangentWeight=0.3949, tangentMode=RCTM_User, time=0, value=0}
           {arriveTangent=1.1137, interpMode=RCIM_Cubic, leaveTangent=1.1137, time=0.3, value=1}
           {arriveTangentWeight=0.3423, interpMode=RCIM_Cubic, leaveTangentWeight=0.0493, tangentMode=RCTM_User, time=0.8979, value=1}
           <at 1 0 RCIM_Cubic>
          ]}}})
     ))
   ) {isSetParameters=true}
   &ScaleColor(
    Scale Alpha=&FloatFromCurve(FloatCurve=DataInterface() {propertyValues {Curve {
        defaultValue=1
        keys [
         {interpMode=RCIM_Cubic, leaveTangentWeight=0.4986, tangentMode=RCTM_User, time=0, value=0}
         {arriveTangentWeight=0.0742, interpMode=RCIM_Cubic, leaveTangentWeight=0.2988, tangentMode=RCTM_User, time=0.1, value=1}
         {interpMode=RCIM_Cubic, tangentMode=RCTM_User, time=0.5026, value=1}
         {arriveTangentWeight=0.1658, interpMode=RCIM_Cubic, tangentMode=RCTM_User, time=0.9145, value=0}
        ]
       }}})
    ScaleA=true
    ScaleRGB=true
   ) {enabled=false}
  ]
  rendererClasses [&NiagaraSpriteRendererProperties]
  renderers [
   {rendererClass=&NiagaraSpriteRendererProperties}
   {rendererClass=&NiagaraSpriteRendererProperties, rendererIndex=1}
  ]
 }
 MagicHand_R {
  emitterUpdateScript [
   &EmitterState()
   &SpawnBurst_Instantaneous(Spawn Count=3, Use Spawn Probability=false, UseLoopCountLimit=false) {enabled=false}
   &SpawnRate(SpawnRate=18, Use Spawn Probability=false)
  ]
  particleSpawnScript [
   &PartitionParticles(Number Of Partitions=1)
   &InitializeParticle(
    Color=&Multiply_LinearColorByFloat(
     Linear Color=<rgba 0.3073 0.7786 1 1>
     Float=&FNTimeOfDayScalar(Night Scale=0.5, Morning Scale=0.78, Evening Scale=0.78)
    )
    Lifetime=0.25
    Sprite Rotation Mode=::NewEnumerator2
    Sprite Size Mode=::NewEnumerator1
    Uniform Sprite Size Max=22
    Uniform Sprite Size Min=20
    UsePositionOffset=false
    Write Material Random=true
   )
   &CameraOffset(Camera Offset Amount=&RandomRangeFloat(Minimum=-15, Maximum=-10))
   /SetVariables(
    Particles.Magic Large Color:LinearColor=<rgba 0.1214 0.2597 0.9323 1>
    Particles.Magic Large Size:Vector2f=&Vector2DFromFloat(Value=&RandomRangeFloat(Minimum=25, Maximum=30))
   ) {isSetParameters=true}
  ]
  particleUpdateScript [
   &ParticleState()
   &ScaleSpriteSize(Uniform Curve Sprite Scale=DataInterface() {propertyValues {Curve {keys [
        {arriveTangentWeight=0.0807, interpMode=RCIM_Cubic, leaveTangentWeight=0.3949, tangentMode=RCTM_User, time=0, value=0}
        {arriveTangent=1.1137, interpMode=RCIM_Cubic, leaveTangent=1.1137, time=0.15, value=1}
        {arriveTangentWeight=0.3423, interpMode=RCIM_Cubic, leaveTangentWeight=0.0493, tangentMode=RCTM_User, time=0.8979, value=1}
        <at 1 0 RCIM_Cubic>
       ]}}})
   &SpriteRotationRate(Rotation Rate=&RandomRangeFloat(Minimum=-100, Maximum=100))
   &SkeletalMeshLocation(
    Bone / Socket Index=&SelectIntFromArray(Direct Array Index=$Particles.PartitionParticles.Partition, Int Selection Array=DataInterface() {propertyValues {IntData [1]}})
    Bone Sampling Mode=::NewEnumerator4
    MaskSpawnGroup=false
    Orientation Sampling=::NewEnumerator1
    Sampled Position Offset=<-10 0 0>
    Skeletal Mesh=$System.Skeletal Mesh
   )
   &SolveForcesAndVelocity(Clamp Velocity=false, Limit Acceleration=false)
   /SetVariables(
    Particles.Scale Magic Large Color:LinearColor=&MakeLinearColorFromVectorAndFloat(
     "Vector (RGB)"=&MakeVectorFromLinearColorRGB(Color=$Particles.Initial.Magic Large Color)
     "Float (Alpha)"=&FloatFromCurve(FloatCurve=DataInterface() {propertyValues {Curve {keys [<at 0 1>,<at 1 1>]}}})
    )
    Particles.Scale Magic Large Size:Vector2f=&Vector2DFromFloat(Value=&Multiply_Float(
      A=&MakeFloatFromVector2d(Vector 2D=$Particles.Initial.Magic Large Size)
      B=&FloatFromCurve(FloatCurve=DataInterface() {propertyValues {Curve {keys [
           {arriveTangentWeight=0.0807, interpMode=RCIM_Cubic, leaveTangentWeight=0.3949, tangentMode=RCTM_User, time=0, value=0}
           {arriveTangent=1.1137, interpMode=RCIM_Cubic, leaveTangent=1.1137, time=0.3, value=1}
           {arriveTangentWeight=0.3423, interpMode=RCIM_Cubic, leaveTangentWeight=0.0493, tangentMode=RCTM_User, time=0.8979, value=1}
           <at 1 0 RCIM_Cubic>
          ]}}})
     ))
   ) {isSetParameters=true}
   &ScaleColor(
    Scale Alpha=&FloatFromCurve(FloatCurve=DataInterface() {propertyValues {Curve {
        defaultValue=1
        keys [
         {interpMode=RCIM_Cubic, leaveTangentWeight=0.4986, tangentMode=RCTM_User, time=0, value=0}
         {arriveTangentWeight=0.0742, interpMode=RCIM_Cubic, leaveTangentWeight=0.2988, tangentMode=RCTM_User, time=0.1, value=1}
         {interpMode=RCIM_Cubic, tangentMode=RCTM_User, time=0.5026, value=1}
         {arriveTangentWeight=0.1658, interpMode=RCIM_Cubic, tangentMode=RCTM_User, time=0.9145, value=0}
        ]
       }}})
    ScaleA=true
    ScaleRGB=true
   ) {enabled=false}
  ]
  rendererClasses [&NiagaraSpriteRendererProperties]
  renderers [
   {rendererClass=&NiagaraSpriteRendererProperties}
   {rendererClass=&NiagaraSpriteRendererProperties, rendererIndex=1}
  ]
 }
}
summary {
 systemName=NS_Emote_SonnetSpirit_MagicHands
 emitters [
  {emitterName=MagicHand_L, rendererClasses [&NiagaraSpriteRendererProperties]}
  {emitterName=MagicHand_R, rendererClasses [&NiagaraSpriteRendererProperties]}
 ]
 userVariables [
  User.bIsFrontEnd {type=NiagaraBool}
  User.bIsFrontEndPreview {type=NiagaraBool}
  User.IsLargeBody {type=NiagaraBool}
  User.OwnerSkeleton {
   type=NiagaraDataInterfaceSkeletalMesh
   defaultValue {
    struct=NiagaraExt_VariableValue_DataInterface
    value {dataInterface=&self:NiagaraDataInterfaceSkeletalMesh_0}
   }
  }
 ]
}
